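# Physics asset named for the SK_Proto_CyberArmor_Female skeletal mesh, stored with the FX of the one-handed female CyberArmor pickaxe.
# It holds a single body on the Anim bone, shaped as one slightly offset capsule rotated about 90 degrees.
# No constraints are defined, so the asset provides one simple physics volume rather than a jointed ragdoll.
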PhysicsAsset /Game/Weapons/FORT_Melee/Pickaxe_CyberArmorFemale1H/FX/SK_Proto_CyberArmor_Female_PhysicsAsset
bodies {Anim [{shapeType=Capsule, length=13.294, radius=8.021, center=<-2.3397 -0.0079 0.038>, rotation=<rot 0.0465 -90 3.0252>}]}
constraints []
